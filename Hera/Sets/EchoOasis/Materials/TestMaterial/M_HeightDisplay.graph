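# Unlit test material in the EchoOasis set that displays world height relative to a WaterHeight parameter.
# A Step of world Z against WaterHeight gives a first-channel mask of what lies at or below the water level.
# The second channel is the saturated height difference divided by a negative WaterHeight Scale, giving a gradient that rises with depth.
# Both values are appended into one two-channel vector.
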Material /Hera/Sets/EchoOasis/Materials/TestMaterial/M_HeightDisplay
graph {
 properties {bAutomaticallySetUsageInEditor=false, shadingModel=MSM_Unlit, usedShadingModels=Unlit}
 nodes [
  n1 = WorldPosition()
  n2 = ScalarParameter() {defaultValue=1500, parameterName=WaterHeight}
  n3 = Subtract(A=@n1.Z, B=@n2)
  n4 = ScalarParameter() {defaultValue=-500, parameterName=WaterHeight Scale}
  AppendVector(A=Step(Y=@n1.Z, X=@n2), B=Saturate(None=Divide(A=@n3, B=@n4) {constB=-1000}))
 ]
}
