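# Translucent lit test material for EchoOasis that builds an animated, three-color striped surface with a swirling center hole, apparently for water.
# A world-projected grayscale depth texture blends with a radial hole mask to drive sine bands, varied by panning Perlin noise.
# Time-animated polar UVs form swirl stripes at the center, and the combined pattern drives color blending and depth-fade distance.
# Several leftover branches, including distance field sampling, feed nothing.
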Material /Hera/Sets/EchoOasis/Materials/TestMaterial/M_SquidTest
graph {
 properties {
  bAutomaticallySetUsageInEditor=false
  blendMode=BLEND_Translucent
  previewMesh=/Engine/Transient.MaterialEditorMeshComponent_11
  usedShadingModels=Default Lit
 }
 nodes [
  n1 = DepthFade(FadeDistance=ScalarParameter() {defaultValue=1000, parameterName=DepthFadeMaskDistance})
  Add(A=Multiply(A=ScalarParameter() {defaultValue=20, parameterName=DepthFadeMaskIntensity}, B=@n1))
  n2 = Subtract(A=WorldPosition().XYZ, B=VectorParameter() {parameterName=RT Coords, defaultValue=<rgba 14000 -15000 0 1>}.RGB)
  n3 = AppendVector(A=ComponentMask(None=@n2) {g=true}, B=Multiply(A=ComponentMask(None=@n2) {r=true}) {constB=-1})
  n4 = ScalarParameter() {defaultValue=10000, parameterName=RT Size}
  n5 = TextureSample(UVs=Add(A=Divide(A=@n3, B=@n4) {constB=10000}) {constB=0.5}) {samplerType=SAMPLERTYPE_LinearGrayscale, texture=&/T_SquidTest_Edit}
  n6 = ScalarParameter() {defaultValue=2, parameterName=UniformDepth Power}
  n7 = Constant() {r=1}
  n8 = Divide(A=@n7) {constB=11}
  n9 = ScalarParameter() {defaultValue=1, parameterName=Blend factor}
  n10 = LinearInterpolate(
   A=@n7
   B=Multiply(A=ScalarParameter() {defaultValue=20, parameterName=UniformDepth Intensity}, B=Power(Base=@n5.R, Exp=@n6))
   Alpha=Multiply(A=Saturate(None=@n8), B=@n9)
  )
  n11 = Subtract(A=TextureCoordinate(), B=Constant() {r=0.5})
  n12 = Multiply(A=@n11, B=@n11)
  n13 = Add(A=ComponentMask(None=@n12) {r=true}, B=ComponentMask(None=@n12) {g=true})
  n14 = Time()
  n15 = Constant()
  n16 = Multiply(A=Add(A=SquareRoot(None=@n13), B=Multiply(A=@n14, B=@n15))) {constB=2}
  n17 = Saturate(None=OneMinus(None=@n16))
  n18 = Multiply(A=ScalarParameter() {defaultValue=3, parameterName=CenterHoleSize}, B=OneMinus(None=@n17))
  n19 = MaterialFunctionCall(Alpha=Saturate(None=OneMinus(None=@n18))) {materialFunction=&SmoothStep}
  n20 = Multiply(A=ScalarParameter() {defaultValue=10, parameterName=CenterHoleIntensity}, B=@n19.Result)
  Add(A=@n10, B=@n20)
  n21 = Arctangent2!(Y=ComponentMask(None=@n11) {r=true}, X=ComponentMask(None=@n11) {g=true})
  n22 = Constant() {r=2}
  n23 = Frac(None=Divide(
    A=ComponentMask(None=@n21) {r=true}
    B=MaterialFunctionCall(Multiplier=@n22) {materialFunction=&Pi}.Result
   ))
  n24 = Multiply(A=Constant() {r=-0.3}, B=@n14)
  n25 = Add(A=@n16) {constB=-0.05}
  n26 = OneMinus(None=Power(Base=OneMinus(None=@n25)) {constExponent=1})
  n27 = Add(A=Multiply(A=Add(A=@n23, B=@n24)), B=Multiply(A=@n26) {constB=33})
  n28 = Frac(None=Multiply(A=Frac(None=@n27)))
  n29 = Multiply(A=@n28) {constB=2}
  Ceil(None=Add(A=@n29) {constB=-1}) {desc=mask}
  Constant()
  n30 = WorldPosition()
  n31 = MaterialFunctionCall() {materialFunction=&ComponentPivotLocation}
  n32 = Divide(A=ComponentMask(None=Subtract(A=@n30.XYZ, B=@n31.Result)) {g=true, r=true}) {constB=5000}
  Distance(A=@n32, B=Constant())
  n33 = AppendVector(A=WorldPosition().XY, B=ScalarParameter() {defaultValue=1500, parameterName=WWater Height})
  Divide(A=DistanceToNearestSurface!(World Position=@n33)) {constB=50}
  Frac(None=@n29) {desc=uv}
  Frac(None=Add(A=@n32) {constB=0.5})
  n34 = VectorParameter() {parameterName=Color A, defaultValue=<rgba 1 0 1 1>}
  n35 = VectorParameter() {parameterName=Color B, defaultValue=<rgba 0.12 1.2 1.2 1>}
  n36 = WorldPosition()
  n37 = Constant() {r=0.05}
  n38 = Time()
  n39 = TextureSample(UVs=Add(A=Divide(A=@n36.XY) {constB=5000}, B=Multiply(A=@n37, B=@n38))) {samplerType=SAMPLERTYPE_Masks, texture=/Game/Characters/Player/Female/Medium/Bodies/F_MED_Neon_Tight_Suit/Materials/Soft_Perlin_output}
  n40 = LinearInterpolate(Alpha=MaterialFunctionCall(In=Power(Base=@n39.R), Contrast=Constant()) {materialFunction=&CheapContrast}.Result) {constA=0.3, constB=0.5}
  n41 = Power(Base=@n19.Result) {constExponent=1}
  n42 = Constant() {r=0.2}
  n43 = ScalarParameter() {defaultValue=1, parameterName=Blend factor centerHole}
  n44 = LinearInterpolate(
   A=@n10
   B=@n20
   Alpha=Multiply(
    A=MaterialFunctionCall(In=@n41, Contrast=@n42) {materialFunction=&CheapContrast}.Result
    B=@n43
   )
  ) {constAlpha=0.9}
  n45 = Constant() {r=0.03}
  n46 = Time()
  n47 = Sine(None=Add(A=Divide(A=@n44) {constB=5}, B=Multiply(A=@n45, B=@n46)))
  n48 = Saturate(None=LinearInterpolate(A=Sine(None=@n47), B=Cosine(None=@n47), Alpha=@n39.R))
  n49 = ScalarParameter() {defaultValue=1, parameterName=Stripes gradient offset}
  n50 = Sine(None=@n28)
  n51 = Constant()
  n52 = Saturate(None=OneMinus(None=MaterialFunctionCall(In=@n50, Contrast=@n51) {materialFunction=&CheapContrast}.Result)) {desc=soft edges}
  n53 = Multiply(A=OneMinus(None=Multiply(A=@n49, B=@n52)), B=OneMinus(None=Power(Base=@n17) {constExponent=22}))
  n54 = Saturate(None=LinearInterpolate(
    A=Multiply(A=@n40, B=@n48) {constB=0.5}
    B=MaterialFunctionCall!(In=@n53) {materialFunction=/Game/Environments/FrontEnd/BackPlates/Textures/Delulu/MF_SmoothStepInverse}."InverseSmoothStep (Aprox)"
    Alpha=Power(Base=@n17) {constExponent=10}
   ))
  LinearInterpolate(
   A=VectorParameter() {parameterName=Color C, defaultValue=<rgba 0.3 0 1.5 1>}.RGB
   B=Power(Base=LinearInterpolate(A=@n34.RGB, B=@n35.RGB, Alpha=@n54))
   Alpha=DepthFade() {fadeDistanceDefault=1000}
  )
  n55 = ScalarParameter() {defaultValue=200, parameterName=Depth Fade}
  n56 = ScalarParameter() {defaultValue=500, parameterName=Depth Fade 2}
  MaterialFunctionCall(Alpha=DepthFade(FadeDistance=LinearInterpolate(A=@n55, B=@n56, Alpha=@n54))) {materialFunction=&SmoothStep}
  n57 = DistanceFieldGradient!(World Position=@n33)
  Multiply(A=Distance(A=ComponentMask(None=@n57) {g=true, r=true}, B=Constant())) {constB=1000}
  Multiply(A=Power(Base=@n16) {constExponent=0.05}) {constB=10}
  OneMinus(None=@n1)
  Power(Base=@n25) {constExponent=5}
  Reroute() {desc=Water Depth}
  Reroute(None=@n43)
  Reroute(None=@n10)
  Reroute(None=@n9)
  Saturate(None=@n26)
 ]
}
